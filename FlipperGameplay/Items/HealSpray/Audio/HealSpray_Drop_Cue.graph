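# Sound cue for dropping the Heal Spray item.
# A Random node picks between two equally weighted drop waves and is excluded from branch culling.
# It plays at reduced volume through the shared weapon-drop attenuation and loot-collect sound class, with the Weapon_Drop_StopOldest concurrency set assigned to limit overlapping drops.
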SoundCue /FlipperGameplay/Items/HealSpray/Audio/HealSpray_Drop_Cue
root=Random_1
nodes [
 {id=Random_1, children [WavePlayer_0,WavePlayer_3], properties {bShouldExcludeFromBranchCulling=true, weights [1,1]}}
 {id=WavePlayer_0, properties {soundWaveAssetPtr=&/HealSpray_Drop_02}}
 {id=WavePlayer_3, properties {soundWaveAssetPtr=&/HealSpray_Drop_01}}
]
properties {
 attenuationSettings=&Fort_Weapon_Drop_Att
 maxDistance=3000
 soundClassObject=&Fort_Loot_Collect
 volumeMultiplier=0.7
 concurrencyOverrides {maxCount=1, resolutionRule=PreventNew}
 concurrencySet [&Weapon_Drop_StopOldest]
}
