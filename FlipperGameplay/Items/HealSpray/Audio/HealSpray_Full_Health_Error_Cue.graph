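# Sound cue for the HealSpray item's error sound, named for use at full health.
# A mixer layers a modulated spray error wave at full volume with a much quieter, briefly randomly delayed pick from five HealSpray pickup waves.
# The cue borrows weapon reload attenuation, plays in the loot collect sound class, and overrides concurrency to one instance with virtualization disabled.
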SoundCue /FlipperGameplay/Items/HealSpray/Audio/HealSpray_Full_Health_Error_Cue
root=Mixer_1
nodes [
 {id=Delay_1, children [Random_0], properties {delayMax=0.15, delayMin=0.05}}
 {id=Mixer_1, children [Modulator_0,Delay_1], properties {inputVolume [1,0.1]}}
 {id=Modulator_0, children [WavePlayer_3], properties {pitchMin=1}}
 {
  id=Random_0
  children [WavePlayer_4,WavePlayer_5,WavePlayer_6,WavePlayer_7,WavePlayer_8]
  properties {bShouldExcludeFromBranchCulling=true, weights [1,1,1,1,1]}
 }
 {id=WavePlayer_3, properties {soundWaveAssetPtr=&/HealSpray_Spray_Error_01}}
 {id=WavePlayer_4, properties {soundWaveAssetPtr=&/HealSpray_Pickup_05}}
 {id=WavePlayer_5, properties {soundWaveAssetPtr=&/HealSpray_Pickup_01}}
 {id=WavePlayer_6, properties {soundWaveAssetPtr=&/HealSpray_Pickup_02}}
 {id=WavePlayer_7, properties {soundWaveAssetPtr=&/HealSpray_Pickup_03}}
 {id=WavePlayer_8, properties {soundWaveAssetPtr=&/HealSpray_Pickup_04}}
]
properties {
 attenuationSettings=&Fort_Weapon_Reload_Att
 bOverrideConcurrency=true
 maxDistance=2250
 soundClassObject=&Fort_Loot_Collect
 virtualizationMode=Disabled
 volumeMultiplier=0.5
 concurrencyOverrides {maxCount=1}
}
