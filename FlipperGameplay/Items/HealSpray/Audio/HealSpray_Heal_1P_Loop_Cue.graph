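# First-person heal spray loop cue, blending a full-canister spray layer with an empty-canister one.
# Two continuous modulators read Spray_Loop in opposite directions: as it falls from its default, the full layer quiets without muting and the empty layer fades in.
# Each layer pairs an enveloped looping wave with a pitch-modulated start one-shot.
# The full layer also mixes in two random picks of pickup sounds, one pitched down.
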SoundCue /FlipperGameplay/Items/HealSpray/Audio/HealSpray_Heal_1P_Loop_Cue
root=Mixer_3
nodes [
 {
  id=Enveloper_0
  children [WavePlayer_0]
  properties {pitchCurve {editorCurveData {keys [<at 0 1.1>, {time=4}]}}, volumeCurve {editorCurveData {keys [<at 0 0>,<at 0.05 1>]}}}
 }
 {
  id=Enveloper_2
  children [WavePlayer_24]
  properties {
   pitchCurve {editorCurveData {keys [
      {arriveTangent=0.2561, interpMode=RCIM_Cubic, leaveTangent=0.2561, time=0, value=1.1}
      {arriveTangent=0.0138, interpMode=RCIM_Cubic, leaveTangent=0.0138, time=0.1086, value=1.1278}
      {arriveTangent=-0.1434, interpMode=RCIM_Cubic, leaveTangent=-0.1434, time=0.5281, value=1.1073}
      {time=1}
     ]}}
   volumeCurve {editorCurveData {keys [{arriveTangent=73.1524, interpMode=RCIM_Cubic, leaveTangent=73.1525, tangentMode=RCTM_User, time=0, value=0}, <at 0.05 1>, <at 10 0.65>]}}
  }
 }
 {id=Mixer_0, children [Enveloper_0,Modulator_0,Random_0,Modulator_1], properties {inputVolume [1.5,1.5,0.5,0.65]}}
 {id=Mixer_2, children [Modulator_5,Modulator_4], properties {inputVolume [2.5,0.65]}}
 {id=Mixer_3, children [ModulatorContinuous_1, ModulatorContinuous_0], properties {inputVolume [1,4]}}
 {
  id=ModulatorContinuous_0
  children [Mixer_2]
  properties {volumeModulationParams {default=30, maxInput=30, maxOutput=0, minOutput=1, parameterName=Spray_Loop}}
 }
 {
  id=ModulatorContinuous_1
  children [Mixer_0]
  properties {volumeModulationParams {default=30, maxInput=30, minOutput=0.3, parameterName=Spray_Loop}}
 }
 {id=Modulator_0, children [WavePlayer_1], properties {pitchMax=1.5, pitchMin=0.9, volumeMin=1}}
 {id=Modulator_1, children [Random_1], properties {pitchMax=0.8, pitchMin=0.75}}
 {id=Modulator_4, children [WavePlayer_25], properties {pitchMax=2.5, pitchMin=2, volumeMax=1, volumeMin=1}}
 {id=Modulator_5, children [Enveloper_2], properties {volumeMax=1, volumeMin=1}}
 {id=Random_0, children [WavePlayer_2,WavePlayer_3,WavePlayer_4,WavePlayer_5,WavePlayer_6], properties {weights [1,1,1,1,1]}}
 {id=Random_1, children [WavePlayer_7, WavePlayer_8, WavePlayer_9, WavePlayer_10, WavePlayer_11], properties {weights [1,1,1,1,1]}}
 {id=WavePlayer_0, properties {bLooping=true, soundWaveAssetPtr=&/HealSpray_Spray_Full_Loop}}
 {id=WavePlayer_1, properties {soundWaveAssetPtr=&/HealSpray_Spray_Full_Start_01}}
 {id=WavePlayer_10, properties {soundWaveAssetPtr=&/HealSpray_Pickup_03}}
 {id=WavePlayer_11, properties {soundWaveAssetPtr=&/HealSpray_Pickup_04}}
 {id=WavePlayer_2, properties {soundWaveAssetPtr=&/HealSpray_Pickup_05}}
 {id=WavePlayer_24, properties {bLooping=true, soundWaveAssetPtr=&/HealSpray_Spray_Empty_Loop}}
 {id=WavePlayer_25, properties {soundWaveAssetPtr=&HealSpray_Spray_Full_Start_01}}
 {id=WavePlayer_3, properties {soundWaveAssetPtr=&/HealSpray_Pickup_01}}
 {id=WavePlayer_4, properties {soundWaveAssetPtr=&/HealSpray_Pickup_02}}
 {id=WavePlayer_5, properties {soundWaveAssetPtr=&HealSpray_Pickup_03}}
 {id=WavePlayer_6, properties {soundWaveAssetPtr=&HealSpray_Pickup_04}}
 {id=WavePlayer_7, properties {soundWaveAssetPtr=&HealSpray_Pickup_05}}
 {id=WavePlayer_8, properties {soundWaveAssetPtr=&HealSpray_Pickup_01}}
 {id=WavePlayer_9, properties {soundWaveAssetPtr=&HealSpray_Pickup_02}}
]
properties {
 bOverrideConcurrency=true
 maxDistance=2097152
 soundClassObject=&Fort_Consumables
 volumeMultiplier=1.25
 concurrencyOverrides {maxCount=3}
}
