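# Looping third-person sound cue for Heal Spray healing, with attenuation and a concurrency limit.
# A "full" branch mixes an enveloped spray loop, a start one-shot and two random pickup layers, while an "empty" branch mixes an enveloped empty loop with a higher-pitched start one-shot.
# Continuous modulators driven by the Spray_Loop parameter crossfade between them: as it falls, the empty branch rises and the full branch drops without going silent.
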SoundCue /FlipperGameplay/Items/HealSpray/Audio/HealSpray_Heal_3P_Loop_Cue
root=Mixer_8
nodes [
 {
  id=Enveloper_4
  children [WavePlayer_16]
  properties {pitchCurve {editorCurveData {keys [<at 0 1.1>, {time=4}]}}, volumeCurve {editorCurveData {keys [<at 0 0>,<at 0.05 1>]}}}
 }
 {
  id=Enveloper_5
  children [WavePlayer_30]
  properties {
   pitchCurve {editorCurveData {keys [
      {arriveTangent=0.2561, interpMode=RCIM_Cubic, leaveTangent=0.2561, time=0, value=1.1}
      {arriveTangent=0.0138, interpMode=RCIM_Cubic, leaveTangent=0.0138, time=0.1086, value=1.1278}
      {arriveTangent=-0.1434, interpMode=RCIM_Cubic, leaveTangent=-0.1434, time=0.5281, value=1.1073}
      {time=1}
     ]}}
   volumeCurve {editorCurveData {keys [{arriveTangent=73.1524, interpMode=RCIM_Cubic, leaveTangent=73.1525, tangentMode=RCTM_User, time=0, value=0}, <at 0.05 1>, <at 10 0.65>]}}
  }
 }
 {id=Mixer_6, children [Enveloper_4,Modulator_2,Random_2,Modulator_3], properties {inputVolume [1.5,1.5,0.5,0.65]}}
 {id=Mixer_7, children [Modulator_7,Modulator_6], properties {inputVolume [2.5,0.65]}}
 {id=Mixer_8, children [ModulatorContinuous_3, ModulatorContinuous_2], properties {inputVolume [1,4]}}
 {
  id=ModulatorContinuous_2
  children [Mixer_7]
  properties {volumeModulationParams {default=30, maxInput=30, maxOutput=0, minOutput=1, parameterName=Spray_Loop}}
 }
 {
  id=ModulatorContinuous_3
  children [Mixer_6]
  properties {volumeModulationParams {default=30, maxInput=30, minOutput=0.3, parameterName=Spray_Loop}}
 }
 {id=Modulator_2, children [WavePlayer_17], properties {pitchMax=1.5, pitchMin=0.9, volumeMin=1}}
 {id=Modulator_3, children [Random_3], properties {pitchMax=0.8, pitchMin=0.75}}
 {id=Modulator_6, children [WavePlayer_31], properties {pitchMax=2.5, pitchMin=2, volumeMax=1, volumeMin=1}}
 {id=Modulator_7, children [Enveloper_5], properties {volumeMax=1, volumeMin=1}}
 {id=Random_2, children [WavePlayer_18, WavePlayer_19, WavePlayer_20, WavePlayer_21, WavePlayer_22], properties {weights [1,1,1,1,1]}}
 {id=Random_3, children [WavePlayer_23, WavePlayer_26, WavePlayer_27, WavePlayer_28, WavePlayer_29], properties {weights [1,1,1,1,1]}}
 {id=WavePlayer_16, properties {bLooping=true, soundWaveAssetPtr=&/HealSpray_Spray_Full_Loop}}
 {id=WavePlayer_17, properties {soundWaveAssetPtr=&/HealSpray_Spray_Full_Start_01}}
 {id=WavePlayer_18, properties {soundWaveAssetPtr=&/HealSpray_Pickup_05}}
 {id=WavePlayer_19, properties {soundWaveAssetPtr=&/HealSpray_Pickup_01}}
 {id=WavePlayer_20, properties {soundWaveAssetPtr=&/HealSpray_Pickup_02}}
 {id=WavePlayer_21, properties {soundWaveAssetPtr=&/HealSpray_Pickup_03}}
 {id=WavePlayer_22, properties {soundWaveAssetPtr=&/HealSpray_Pickup_04}}
 {id=WavePlayer_23, properties {soundWaveAssetPtr=&HealSpray_Pickup_05}}
 {id=WavePlayer_26, properties {soundWaveAssetPtr=&HealSpray_Pickup_01}}
 {id=WavePlayer_27, properties {soundWaveAssetPtr=&HealSpray_Pickup_02}}
 {id=WavePlayer_28, properties {soundWaveAssetPtr=&HealSpray_Pickup_03}}
 {id=WavePlayer_29, properties {soundWaveAssetPtr=&HealSpray_Pickup_04}}
 {id=WavePlayer_30, properties {bLooping=true, soundWaveAssetPtr=&/HealSpray_Spray_Empty_Loop}}
 {id=WavePlayer_31, properties {soundWaveAssetPtr=&HealSpray_Spray_Full_Start_01}}
]
properties {
 attenuationSettings=&/HealSpray_Att
 bOverrideConcurrency=true
 maxDistance=8900
 soundClassObject=&Fort_Consumables
 volumeMultiplier=1.25
 concurrencyOverrides {maxCount=3}
}
