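# Sound cue for the third-person stop sound of the HealSpray item's healing loop.
# A single Player_Heal_Loop_Stop wave plays through a modulator that varies pitch while holding volume constant.
# It uses the HealSpray attenuation asset and the Fort consumables sound class, with a concurrency override capping simultaneous instances.
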SoundCue /FlipperGameplay/Items/HealSpray/Audio/HealSpray_Heal_3P_Loop_Stop_Cue
root=Modulator_0
nodes [
 {id=Modulator_0, children [WavePlayer_2], properties {pitchMax=1.15, volumeMax=1, volumeMin=1}}
 {id=WavePlayer_2, properties {soundWaveAssetPtr=&/Player_Heal_Loop_Stop_01}}
]
properties {
 attenuationSettings=&/HealSpray_Att
 bOverrideConcurrency=true
 maxDistance=8900
 soundClassObject=&Fort_Consumables
 volumeMultiplier=1
 concurrencyOverrides {maxCount=3}
}
