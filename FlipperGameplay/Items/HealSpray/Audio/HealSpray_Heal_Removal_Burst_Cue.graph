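# Sound cue for the removal burst of the HealSpray item's heal, classed as a Fort consumable sound.
# A single burst wave plays through a modulator that randomizes pitch while holding volume constant.
# The cue uses the HealSpray attenuation asset, slightly boosts volume, and overrides concurrency to cap simultaneous instances at a small count.
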SoundCue /FlipperGameplay/Items/HealSpray/Audio/HealSpray_Heal_Removal_Burst_Cue
root=Modulator_5
nodes [
 {id=Modulator_5, children [WavePlayer_6], properties {pitchMax=1.2, volumeMax=1, volumeMin=1}}
 {id=WavePlayer_6, properties {soundWaveAssetPtr=&/HealSpray_Heal_Removal_Burst_01}}
]
properties {
 attenuationSettings=&/HealSpray_Att
 bOverrideConcurrency=true
 maxDistance=8900
 soundClassObject=&Fort_Consumables
 volumeMultiplier=1.2
 concurrencyOverrides {maxCount=3}
}
