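# Sound cue for the landing sound of the HealSpray item, presumably when it is dropped.
# A single Flopper landed wave borrowed from the Athena item sounds feeds a Modulator, which by design randomizes pitch and volume per play.
# The cue raises pitch and lowers volume overall, and uses weapon-drop attenuation with the loot-collect sound class.
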SoundCue /FlipperGameplay/Items/HealSpray/Audio/HealSpray_Landed_Cue
root=Modulator_1
nodes [
 {id=Modulator_1, children [WavePlayer_2]}
 {id=WavePlayer_2, properties {soundWaveAssetPtr=/Game/Athena/Sounds/Items/Flopper/Flopper_Landed_01}}
]
properties {
 attenuationSettings=&Fort_Weapon_Drop_Att
 maxDistance=3000
 pitchMultiplier=1.2
 soundClassObject=&Fort_Loot_Collect
 volumeMultiplier=0.65
}
