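# Sound cue for the HealSpray item running out of ammo, built from throw, rattle, cap pop and drop sounds.
# A root mixer schedules two delayed layers: a random whoosh first, then a blend of random cap pop, rattle and drop, with the drop pitch and volume modulated.
# The mixer's first input is empty.
# It uses HealSpray attenuation, the Fort_Consumables sound class and a concurrency override capping simultaneous instances.
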SoundCue /FlipperGameplay/Items/HealSpray/Audio/HealSpray_OutOfAmmo_Cue
root=Mixer_0
nodes [
 {id=Delay_1, children [Random_3], properties {delayMax=0.35, delayMin=0.35}}
 {id=Delay_2, children [Mixer_1], properties {delayMax=1, delayMin=1}}
 {id=Mixer_0, children [null,Delay_1,Delay_2], properties {inputVolume [0.4,1,0.6]}}
 {id=Mixer_1, children [Random_4,Random_2,Modulator_0], properties {inputVolume [1,0.7,0.6]}}
 {id=Modulator_0, children [Random_5], properties {pitchMax=1.25, volumeMin=0.65}}
 {
  id=Random_2
  children [WavePlayer_12, WavePlayer_13, WavePlayer_14, WavePlayer_15]
  properties {bShouldExcludeFromBranchCulling=true, weights [1,1,1,1]}
 }
 {id=Random_3, children [WavePlayer_16, WavePlayer_17, WavePlayer_18], properties {bShouldExcludeFromBranchCulling=true, weights [1,1,1]}}
 {id=Random_4, children [WavePlayer_19, WavePlayer_20, WavePlayer_21], properties {weights [1,1,1]}}
 {id=Random_5, children [WavePlayer_22, WavePlayer_23], properties {bShouldExcludeFromBranchCulling=true, weights [1,1]}}
 {id=WavePlayer_12, properties {soundWaveAssetPtr=&/HealSpray_Throw_Rattle_01}}
 {id=WavePlayer_13, properties {soundWaveAssetPtr=&/HealSpray_Throw_Rattle_02}}
 {id=WavePlayer_14, properties {soundWaveAssetPtr=&/HealSpray_Throw_Rattle_03}}
 {id=WavePlayer_15, properties {soundWaveAssetPtr=&/HealSpray_Throw_Rattle_04}}
 {id=WavePlayer_16, properties {soundWaveAssetPtr=&/HealSpray_Throw_Whoosh_02}}
 {id=WavePlayer_17, properties {soundWaveAssetPtr=&/HealSpray_Throw_Whoosh_03}}
 {id=WavePlayer_18, properties {soundWaveAssetPtr=&/HealSpray_Throw_Whoosh_01}}
 {id=WavePlayer_19, properties {soundWaveAssetPtr=&/HealSpray_Cap_Pop_01}}
 {id=WavePlayer_20, properties {soundWaveAssetPtr=&/HealSpray_Cap_Pop_02}}
 {id=WavePlayer_21, properties {soundWaveAssetPtr=&/HealSpray_Cap_Pop_03}}
 {id=WavePlayer_22, properties {soundWaveAssetPtr=&/HealSpray_Drop_01}}
 {id=WavePlayer_23, properties {soundWaveAssetPtr=&/HealSpray_Drop_02}}
]
properties {
 attenuationSettings=&/HealSpray_Att
 bOverrideConcurrency=true
 maxDistance=8900
 soundClassObject=&Fort_Consumables
 volumeMultiplier=1
 concurrencyOverrides {maxCount=3}
}
