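# Sound cue for picking up the Heal Spray item.
# A mixer layers two equally weighted random picks: one of five pickup variations and one of three cap-pop variations.
# It uses weapon pickup attenuation and the loot collect sound class at reduced volume, with a concurrency override that allows one instance and blocks new ones while it plays.
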SoundCue /FlipperGameplay/Items/HealSpray/Audio/HealSpray_Pickup_Cue
root=Mixer_0
nodes [
 {id=Mixer_0, children [Random_0,Random_1], properties {inputVolume [1,1]}}
 {
  id=Random_0
  children [WavePlayer_0,WavePlayer_1,WavePlayer_2,WavePlayer_3,WavePlayer_5]
  properties {bShouldExcludeFromBranchCulling=true, weights [1,1,1,1,1]}
 }
 {id=Random_1, children [WavePlayer_6,WavePlayer_7,WavePlayer_8], properties {bShouldExcludeFromBranchCulling=true, weights [1,1,1]}}
 {id=WavePlayer_0, properties {soundWaveAssetPtr=&/HealSpray_Pickup_01}}
 {id=WavePlayer_1, properties {soundWaveAssetPtr=&/HealSpray_Pickup_02}}
 {id=WavePlayer_2, properties {soundWaveAssetPtr=&/HealSpray_Pickup_03}}
 {id=WavePlayer_3, properties {soundWaveAssetPtr=&/HealSpray_Pickup_04}}
 {id=WavePlayer_5, properties {soundWaveAssetPtr=&/HealSpray_Pickup_05}}
 {id=WavePlayer_6, properties {soundWaveAssetPtr=&/HealSpray_Cap_Pop_01}}
 {id=WavePlayer_7, properties {soundWaveAssetPtr=&/HealSpray_Cap_Pop_02}}
 {id=WavePlayer_8, properties {soundWaveAssetPtr=&/HealSpray_Cap_Pop_03}}
]
properties {
 attenuationSettings=&Fort_Weapon_Pickup_Att
 bOverrideConcurrency=true
 maxDistance=3500
 soundClassObject=&Fort_Loot_Collect
 volumeMultiplier=0.338
 concurrencyOverrides {maxCount=1, resolutionRule=PreventNew}
}
