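# Sound cue for removing the HealSpray consumable, layering spray-stop, handling and cap-pop sounds.
# A mixer combines a random spray stop, a loop-stop tail, a quieter pitched-down random pickup sound and a briefly delayed, pitched-down random cap pop.
# Modulators vary pitch per layer, and a quality switch plays the pickup layer only on the first two quality levels.
# It uses HealSpray attenuation, the consumables sound class and a small concurrency cap.
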SoundCue /FlipperGameplay/Items/HealSpray/Audio/HealSpray_Removal_Cue
root=Mixer_2
nodes [
 {id=Delay_0, children [Modulator_1], properties {delayMin=0.15}}
 {id=Mixer_2, children [Modulator_2, QualityLevel_0, Modulator_0, Delay_0], properties {inputVolume [1,0.2,1,0.6]}}
 {id=Modulator_0, children [WavePlayer_1], properties {pitchMax=1.1, volumeMax=1, volumeMin=1}}
 {id=Modulator_1, children [Random_2], properties {pitchMax=0.9, pitchMin=0.8}}
 {id=Modulator_2, children [Random_0], properties {pitchMax=1.1, pitchMin=0.9, volumeMin=1}}
 {id=Modulator_3, children [Random_1], properties {pitchMax=0.8, pitchMin=0.75}}
 {id=QualityLevel_0, children [Modulator_3,Modulator_3,null,null]}
 {id=Random_0, children [WavePlayer_0,WavePlayer_2], properties {bShouldExcludeFromBranchCulling=true, weights [1,1]}}
 {id=Random_1, children [WavePlayer_7, WavePlayer_8, WavePlayer_9, WavePlayer_10, WavePlayer_11], properties {weights [1,1,1,1,1]}}
 {id=Random_2, children [WavePlayer_3,WavePlayer_4,WavePlayer_5], properties {bShouldExcludeFromBranchCulling=true, weights [1,1,1]}}
 {id=WavePlayer_0, properties {soundWaveAssetPtr=&/HealSpray_Spray_Stop_02}}
 {id=WavePlayer_1, properties {soundWaveAssetPtr=&/HealSpray_Spray_Loop_Stop_01}}
 {id=WavePlayer_10, properties {soundWaveAssetPtr=&/HealSpray_Pickup_03}}
 {id=WavePlayer_11, properties {soundWaveAssetPtr=&/HealSpray_Pickup_04}}
 {id=WavePlayer_2, properties {soundWaveAssetPtr=&/HealSpray_Spray_Stop_01}}
 {id=WavePlayer_3, properties {soundWaveAssetPtr=&/HealSpray_Cap_Pop_01}}
 {id=WavePlayer_4, properties {soundWaveAssetPtr=&/HealSpray_Cap_Pop_02}}
 {id=WavePlayer_5, properties {soundWaveAssetPtr=&/HealSpray_Cap_Pop_03}}
 {id=WavePlayer_7, properties {soundWaveAssetPtr=&/HealSpray_Pickup_05}}
 {id=WavePlayer_8, properties {soundWaveAssetPtr=&/HealSpray_Pickup_01}}
 {id=WavePlayer_9, properties {soundWaveAssetPtr=&/HealSpray_Pickup_02}}
]
properties {
 attenuationSettings=&/HealSpray_Att
 bOverrideConcurrency=true
 maxDistance=8900
 soundClassObject=&Fort_Consumables
 volumeMultiplier=1
 concurrencyOverrides {maxCount=3}
}
