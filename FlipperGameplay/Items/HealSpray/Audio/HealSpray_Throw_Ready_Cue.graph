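# Sound cue for readying a HealSpray item throw, layering an equip sound with a slightly delayed cap pop.
# A mixer blends two branches, each a modulator over a random pick of three wave variants, with the cap pop branch offset by a short delay.
# It plays at reduced volume with weapon reload attenuation, the loot collect sound class and its own concurrency cap.
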SoundCue /FlipperGameplay/Items/HealSpray/Audio/HealSpray_Throw_Ready_Cue
root=Mixer_0
nodes [
 {id=Delay_0, children [Modulator_1], properties {delayMin=0.15}}
 {id=Mixer_0, children [Modulator_0,Delay_0], properties {inputVolume [1,1]}}
 {id=Modulator_0, children [Random_0]}
 {id=Modulator_1, children [Random_1]}
 {id=Random_0, children [WavePlayer_0,WavePlayer_2,WavePlayer_3], properties {bShouldExcludeFromBranchCulling=true, weights [1,1,1]}}
 {id=Random_1, children [WavePlayer_1,WavePlayer_4,WavePlayer_5], properties {bShouldExcludeFromBranchCulling=true, weights [1,1,1]}}
 {id=WavePlayer_0, properties {soundWaveAssetPtr=&/HealSpray_Equip_01}}
 {id=WavePlayer_1, properties {soundWaveAssetPtr=&/HealSpray_Cap_Pop_01}}
 {id=WavePlayer_2, properties {soundWaveAssetPtr=&/HealSpray_Equip_02}}
 {id=WavePlayer_3, properties {soundWaveAssetPtr=&/HealSpray_Equip_03}}
 {id=WavePlayer_4, properties {soundWaveAssetPtr=&/HealSpray_Cap_Pop_02}}
 {id=WavePlayer_5, properties {soundWaveAssetPtr=&/HealSpray_Cap_Pop_03}}
]
properties {
 attenuationSettings=&Fort_Weapon_Reload_Att
 bOverrideConcurrency=true
 maxDistance=2250
 soundClassObject=&Fort_Loot_Collect
 volumeMultiplier=0.5
 concurrencyOverrides {maxCount=3}
}
